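# Translucent unlit Niagara sprite material that draws a distorted ring ripple for the HayBrush backpack effects.
# Radially mapped, panning noise offsets the UVs, and a SphereMask on their distance from center forms the ring, with dynamic parameters driving radius, thickness and distortion.
# A circular edge mask, particle alpha and a movable mask shape opacity; that mask uses Mask X for both axes, leaving Mask Y unused.
# Particle color times intensity feeds emissive.
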Material /BRCosmetics/Accessories/FORT_Backpacks/Backpack_F_MED_HayBrush/FX/M_Backpack_HayBrush_Ripple
graph {
 properties {
  bUsedWithNiagaraSprites=true
  blendMode=BLEND_Translucent
  pixelDepthOffsetMode=PDOM_Legacy
  shadingModel=MSM_Unlit
  translucencyLightingMode=TLM_VolumetricPerVertexNonDirectional
  usedShadingModels=Unlit
 }
 nodes [
  n1 = ParticleColor()
  MaterialFunctionCall(Emissive=Multiply(A=@n1.RGB, B=ScalarParameter() {defaultValue=1, parameterName=Emissive_Intensity})) {materialFunction=&MF_FX_EmissiveAdjustments}
  n2 = TextureCoordinate()
  n3 = Add(A=TextureCoordinate()) {constB=-0.5}
  n4 = TextureCoordinate()
  n5 = TextureSample!(
   UVs=Panner(Coordinate=MaterialFunctionCall(Vector or UVs=@n3) {materialFunction=&VectorToRadialValue}.Radial Coordinates) {speedY=-0.5}
   Tex
   "DDX(UVs)"=DDX(Value=@n4)
   "DDY(UVs)"=DDY(Value=@n4)
   Apply View MipBias
  ) {mipValueMode=TMVM_Derivative, samplerType=SAMPLERTYPE_Grayscale, texture=&LowResBlurredNoise}
  n6 = Multiply(
   A=@n5.R
   B=Multiply(
    A=ScalarParameter() {defaultValue=0.1, parameterName=Distortion Intensity, sliderMax=0.5}
    B=DynamicParameter() {defaultValue=<rgba 1 1 0.3 1>, paramNames [Thickness, Distortion, Radius, Range Max]}.Distortion
   )
  ) {constB=0.05}
  n7 = SphereMask(
   A=Distance(A=Add(A=@n2, B=@n6), B=Constant() {r=0.5})
   B=DynamicParameter() {defaultValue=<rgba 1 1 0.3 1>, paramNames [Thickness, Distortion, Radius, Range Max]}.Radius
   Radius=Multiply(A=Constant() {r=0.05}, B=DynamicParameter() {defaultValue=<rgba 1 1 0.3 1>, paramNames [Thickness, Distortion, Radius, Range Max]}.Thickness)
  ) {attenuationRadius=0.05, hardnessPercent=0}
  n8 = SphereMask(A=TextureCoordinate(), B=Constant2Vector() {g=0.5, r=0.5}, Radius=Constant() {r=0.5}, Hardness=Constant())
  n9 = DynamicParameter() {parameterIndex=1, defaultValue=<rgba 0.5 0.5 1 1>, paramNames [Mask X, Mask Y, Radius, Param4]}
  Multiply(
   A=Multiply(A=@n1.A, B=Multiply(A=@n7, B=@n8))
   B=SphereMask(A=TextureCoordinate(), B=AppendVector(A=@n9.Mask X, B=@n9.Mask X), Radius=@n9.Radius) {attenuationRadius=0.5}
  )
 ]
}
