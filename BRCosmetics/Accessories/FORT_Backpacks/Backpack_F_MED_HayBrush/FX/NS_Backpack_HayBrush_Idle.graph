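# Idle Niagara effect for the HayBrush backpack cosmetic, producing bubbles, ripples and pops on what appears to be a liquid surface.
# Bubble_Ripple_Burst spawns in a socket-aligned disc and drives mesh and sprite renderers from one particle, using separate curve-driven ripple and end-of-life burst attributes.
# Ripples_Frog_and_Spoon places socket-relative ripple sprites at three fixed offsets, varied by spawn group.
# Front-end modifiers switch colors and offsets, and time-of-day color tints both emitters.
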NiagaraSystem /BRCosmetics/Accessories/FORT_Backpacks/Backpack_F_MED_HayBrush/FX/NS_Backpack_HayBrush_Idle
dependencies {
 usedDataInterfaces [
  &NiagaraDataInterfaceArrayColor
  &NiagaraDataInterfaceCurve
  &NiagaraDataInterfaceArrayFloat3
  &NiagaraDataInterfaceSkeletalMesh
 ]
 usedDynamicInputs []
 usedModules [
  &self:SystemSpawnScript.SystemScriptSource.SystemScriptGraph.NiagaraNodeAssignment_0.SetVariables_#1
  &SystemState
 ]
}
emitters {
 Bubble_Ripple_Burst {
  emitterUpdateScript [
   &EmitterState()
   &SpawnRate(Spawn Probability=0.3, SpawnRate=10, Use Spawn Probability=true)
  ]
  particleSpawnScript [
   /SetVariables(Particles.MainColor:LinearColor=&SelectLinearColorFromArray(
     Array Sampling Mode=::NewEnumerator1
     Direct Array Index=&MakeIntFromFloat(Float=&FN_FrontEndFloatModifier(Gameplay=0, Front End=1))
     Color Selection Array=DataInterface() {propertyValues {ColorData [<rgba 0.432 0.2947 0.1024 1>, <rgba 0.3169 0.3281 0.1931 1>]}}
     Random Seed=?
    )) {isSetParameters=true, module=SetVariables_#2}
   &InitializeParticle(
    Color=$Particles.MainColor
    Lifetime Max=1.5
    Lifetime Min=1.3
    Lifetime Mode=::NewEnumerator1
    Mesh Scale Mode=::NewEnumerator1
    Mesh Uniform Scale Max=0.025
    Mesh Uniform Scale Min=0.015
    Sprite Rotation Mode=::NewEnumerator2
    Sprite Size Max=<7 7>
    Sprite Size Mode=::NewEnumerator2
    UsePositionOffset=false
    Write Material Random=true
   )
   &CameraOffset(Camera Offset Amount=&FN_FrontEndFloatModifier(Gameplay=0.2, Front End=0.25))
   &InitialMeshOrientation(
    Mesh Orientation Mode=::NewEnumerator2
    Rotation=&MakeVector(X=&RandomRangeFloat(Minimum=-0.1, Maximum=0.1), Y=0, Z=0)
    Use Rotation Vector=true
   )
   &SkeletalMeshLocation(
    Bone / Socket Index=1
    Bone Sampling Mode=::NewEnumerator4
    Interpolate Mesh Sampling=false
    MaskSpawnGroup=false
    Orientation Sampling=::NewEnumerator1
    Sampled Orientation Coordinate Space=::Local
    Sampled Position Coordinate Space=::Local
    Sampled Position Offset=&MakeVector(X=&FN_FrontEndFloatModifier(Gameplay=0.2, Front End=0.7), Y=1, Z=0)
    Skeletal Mesh=$System.Skeletal Mesh
   ) {module=SkeletalMeshLocation001}
   &ShapeLocation(
    Cylinder Height=0
    Cylinder Radius=&FN_FrontEndFloatModifier(Gameplay=5, Front End=5)
    MaskSpawnGroup=false
    Override Seed=false
    Rotation Axis=&RotateVectorByQuaternion(Vector To Rotate=<1 0.21 0>, Quaternion=$Particles.SkeletalMeshLocation001.SampledRotation)
    Rotation Coordinate Space=::Simulation
    Rotation Mode=::NewEnumerator1
    Shape Primitive=::NewEnumerator1
    Surface Only=false
    UseOwnerScale=true
   )
   /SetVariables(Particles.Position=&TransformPosition(Position=$Particles.ShapeLocation.ShapePosition, Source Space=::Local, Destination Space=::Simulation)) {isSetParameters=true}
  ]
  particleUpdateScript [
   &ParticleState()
   &KillParticlesInVolume(Origin Offset:Vector3f=<-3 -7 0>, Sphere Radius=8)
   &SkeletalMeshLocation(
    Bone / Socket Index=1
    Bone Sampling Mode=::NewEnumerator4
    MaskSpawnGroup=false
    Orientation Sampling=::NewEnumerator1
    Position Sampling=::NewEnumerator2
    Skeletal Mesh=$System.Skeletal Mesh
   )
   &SpriteFacingAndAlignment(
    Sprite Facing=&RotateVectorByQuaternion(Vector To Rotate=<1 0.2 0>, Quaternion=$Particles.SkeletalMeshLocation.SampledRotation)
    WriteAlignment=false
    WriteFacing=true
   )
   &SpriteRotationRate(Rotation Rate=20)
   &ScaleMeshSize(Scale Factor=&VectorFromFloat(Value=&FloatFromCurve(FloatCurve=DataInterface() {propertyValues {Curve {keys [
          {arriveTangent=2.7087, interpMode=RCIM_Cubic, leaveTangent=2.7087, tangentMode=RCTM_User, time=0, value=0}
          {arriveTangent=-0.0108, interpMode=RCIM_Cubic, leaveTangent=-0.0108, tangentMode=RCTM_User, time=0.5, value=0.5}
          {arriveTangent=1.2003, interpMode=RCIM_Cubic, leaveTangent=1.2003, tangentMode=RCTM_User, time=0.9, value=0.7}
          {interpMode=RCIM_Cubic, leaveTangent=0.625, time=0.91, value=0}
         ]}}})))
   &PointForce(
    Force Origin=&AddVectorToPosition(Position=/Niagara/DynamicInputs/Helpers/SimulationPosition(), Vector=<0 0 -6>)
    Force Strength=&FloatFromCurve(
     FloatCurve=DataInterface() {propertyValues {Curve {keys [
         <at 0.2 0 RCIM_Cubic>
         {interpMode=RCIM_Cubic, tangentMode=RCTM_User, time=0.3, value=1}
         {interpMode=RCIM_Cubic, tangentMode=RCTM_User, time=0.4, value=0}
        ]}}}
     Scale Curve=&RandomRangeFloat(Minimum=400, Maximum=600)
    )
    Module.Use Falloff=false
    Module.UseLinearForce=false
   )
   &CurlNoiseForce(
    Noise Strength=&FloatFromCurve(
     FloatCurve=DataInterface() {propertyValues {Curve {keys [{interpMode=RCIM_Cubic, tangentMode=RCTM_User, time=0.3, value=0}, {interpMode=RCIM_Cubic, tangentMode=RCTM_User, time=1, value=1}]}}}
     Scale Curve=100
    )
    Pan Noise=false
   )
   &Drag(Drag=5, Use Linear Drag=true, Use Rotational Drag=true)
   &GravityForce(Gravity=&MakeVector(
     X=0
     Y=0
     Z=&FloatFromCurve(
      FloatCurve=DataInterface() {propertyValues {Curve {keys [<at 0 0>,<at 0.2 0>,<at 0.21 1>]}}}
      Scale Curve=&RandomRangeFloat(Minimum=5, Maximum=10)
     )
    ))
   &PointAttractionForce(
    Attraction Strength=&FloatFromCurve(
     FloatCurve=DataInterface() {propertyValues {Curve {keys [{interpMode=RCIM_Cubic, tangentMode=RCTM_User, time=0, value=1}, {interpMode=RCIM_Cubic, tangentMode=RCTM_User, time=1, value=0}]}}}
     Scale Curve=3
    )
    Kill Within Radius=false
    Module.Use Falloff=true
    UseAttractorOffset=false
   )
   &InheritSourceMovement(
    Applied Position Scale=&VectorFromFloat(Value=&FloatFromCurve(FloatCurve=DataInterface() {propertyValues {Curve {keys [
          {interpMode=RCIM_Cubic, tangentMode=RCTM_User, time=0, value=1}
          {arriveTangent=-0.4868, interpMode=RCIM_Cubic, leaveTangent=-0.4868, tangentMode=RCTM_User, time=1, value=0.7}
         ]}}}))
    ApplyForce=false
    ApplyPosition=true
    ApplyPositionFalloff=false
   )
   &SolveForcesAndVelocity(Clamp Velocity=false, Limit Acceleration=false)
   &DynamicMaterialParameters(
    Index 0 Param 1=&FloatFromCurve(
     FloatCurve=DataInterface() {propertyValues {Curve {keys [
         {arriveTangent=6.1038, interpMode=RCIM_Cubic, leaveTangent=6.1038, tangentMode=RCTM_User, time=0, value=0.25}
         {arriveTangent=0.038, interpMode=RCIM_Cubic, leaveTangent=0.038, tangentMode=RCTM_User, time=0.2, value=1}
         {arriveTangent=-0.0405, interpMode=RCIM_Cubic, leaveTangent=-0.0405, tangentMode=RCTM_User, time=1, value=0}
        ]}}}
     Scale Curve=&RandomRangeFloat(Minimum=0.8, Maximum=1.2)
    )
    Index 0 Param 2=&FloatFromCurve(
     FloatCurve=DataInterface() {propertyValues {Curve {keys [
         {arriveTangent=2.0212, interpMode=RCIM_Cubic, leaveTangent=2.0212, tangentMode=RCTM_User, time=0, value=0}
         {interpMode=RCIM_Cubic, tangentMode=RCTM_User, time=1, value=1}
        ]}}}
     Scale Curve=1.5
    )
    Index 0 Param 3=&FloatFromCurve(
     FloatCurve=DataInterface() {propertyValues {Curve {keys [
         {arriveTangent=2.9028, interpMode=RCIM_Cubic, leaveTangent=2.9028, tangentMode=RCTM_User, time=0, value=0}
         {interpMode=RCIM_Cubic, tangentMode=RCTM_User, time=1, value=1}
        ]}}}
     Scale Curve=0.3
    )
    Module.Index0Param0Write=true
    Module.Index0Param1Write=true
    Module.Index0Param2Write=true
    Module.Index0Param3Write=true
   )
   &ScaleColor(
    Scale Alpha=&FloatFromCurve(FloatCurve=DataInterface() {propertyValues {Curve {keys [
         {interpMode=RCIM_Cubic, tangentMode=RCTM_User, time=0.8, value=1}
         {arriveTangent=-3.602, interpMode=RCIM_Cubic, leaveTangent=-3.602, tangentMode=RCTM_User, time=0.9, value=0.8}
        ]}}})
    Scale RGB=&MakeVectorFromLinearColorRGB(Color=&FNTimeOfDayColor(
      Night Color=<rgba 0.26 0.4943 1 1>
      Morning Color=<rgba 0.8 0.5397 0.516 1>
      Day Color=<rgba 1 1 1 1>
      Evening Color=<rgba 0.8021 0.5013 0.5013 1>
     ))
    ScaleA=true
    ScaleRGB=true
   )
   /SetVariables(
    Particles.Ripple Color:LinearColor=&MakeLinearColorFromVectorAndFloat(
     "Vector (RGB)"=&MakeVectorFromLinearColorRGB(Color=$Particles.Color)
     "Float (Alpha)"=&FloatFromCurve(
      FloatCurve=DataInterface() {propertyValues {Curve {keys [{interpMode=RCIM_Cubic, tangentMode=RCTM_User, time=0, value=1}, {interpMode=RCIM_Cubic, tangentMode=RCTM_User, time=1, value=0}]}}}
      Scale Curve=&RandomRangeFloat(Minimum=0.6, Maximum=0.8)
     )
    )
    Particles.RipplePosition:NiagaraPosition=&TransformPosition(Position=$Particles.ShapeLocation.ShapePosition, Source Space=::Local, Destination Space=::Simulation)
   ) {isSetParameters=true, module=SetVariables_#3}
   /SetVariables(
    Particles.Burst Color:LinearColor=&MakeLinearColorFromVectorAndFloat(
     "Vector (RGB)"=&MakeVectorFromLinearColorRGB(Color=$Particles.Color)
     "Float (Alpha)"=&FloatFromCurve(FloatCurve=DataInterface() {propertyValues {Curve {keys [
          {arriveTangent=85.7303, interpMode=RCIM_Cubic, leaveTangent=85.7303, tangentMode=RCTM_User, time=0.89, value=0}
          {arriveTangent=-0.0213, interpMode=RCIM_Cubic, leaveTangent=-0.0213, tangentMode=RCTM_User, time=0.9, value=1}
          {arriveTangent=-10.9514, interpMode=RCIM_Cubic, leaveTangent=-10.9514, tangentMode=RCTM_User, time=1, value=0.2}
         ]}}})
    )
    Particles.Burst Sprite Size:Vector2f=&Multiply_Vector2D_ByFloat(
     Float=&FloatFromCurve(
      FloatCurve=DataInterface() {propertyValues {Curve {keys [
          {interpMode=RCIM_Cubic, tangentMode=RCTM_User, time=0.89, value=0}
          {arriveTangent=12.5591, interpMode=RCIM_Cubic, leaveTangent=12.5591, tangentMode=RCTM_User, time=0.9, value=0.6}
          <at 1 1.2 RCIM_Cubic>
         ]}}}
      Scale Curve=&RandomRangeFloat(Minimum=1.1, Maximum=1.2)
     )
     Vector2D=<10 10>
    )
    Particles.BurstDynamicParameter:Vector4f=&MakeVector4(
     X=&FloatFromCurve(FloatCurve=DataInterface() {propertyValues {Curve {keys [{interpMode=RCIM_Cubic, tangentMode=RCTM_User, time=0.9, value=0.2}, {interpMode=RCIM_Cubic, tangentMode=RCTM_User, time=1, value=1}]}}})
     Y=0
     Z=0
     W=0
    )
    Particles.BurstPosition:NiagaraPosition=&AddVectorToPosition(
     Position=$Particles.Position
     Vector=&Multiply_VectorByFloat(
      Vector=<0 0 -1>
      Float=&FloatFromCurve(
       FloatCurve=DataInterface() {propertyValues {Curve {keys [
           {interpMode=RCIM_Cubic, tangentMode=RCTM_User, time=0.9, value=0}
           {arriveTangent=18.2753, interpMode=RCIM_Cubic, leaveTangent=18.2753, tangentMode=RCTM_User, time=1, value=1}
          ]}}}
       Scale Curve=2
      )
     )
    )
   ) {isSetParameters=true}
  ]
  rendererClasses [&NiagaraMeshRendererProperties, &NiagaraSpriteRendererProperties]
  renderers [
   {rendererClass=&NiagaraMeshRendererProperties, rendererIndex=0}
   {rendererClass=&NiagaraSpriteRendererProperties, rendererIndex=1}
   {rendererClass=&NiagaraSpriteRendererProperties, rendererIndex=2}
  ]
 }
 Ripples_Frog_and_Spoon {
  emitterUpdateScript [
   &EmitterState()
   &SpawnRate(Spawn Probability=0.8, Use Spawn Probability=false)
  ]
  particleSpawnScript [
   /SetVariables(Particles.MainColor:LinearColor=&SelectLinearColorFromArray(
     Array Sampling Mode=::NewEnumerator1
     Direct Array Index=&MakeIntFromFloat(Float=&FN_FrontEndFloatModifier(Gameplay=0, Front End=1))
     Color Selection Array=DataInterface() {propertyValues {ColorData [<rgba 0.432 0.2947 0.1024 1>, <rgba 0.3169 0.3281 0.1931 1>]}}
     Random Seed=?
    )) {isSetParameters=true, module=SetVariables_#2}
   /SetVariables(Particles.SpawnGroup=&RandomRangeInt(Minimum=0, Maximum=2)) {isSetParameters=true}
   &InitializeParticle(
    Color=$Particles.MainColor
    Lifetime Max=2.25
    Lifetime Min=1.75
    Lifetime Mode=::NewEnumerator1
    Sprite Rotation Angle=&MaskFloatBySpawnGroup(Masked Float=0, Passthrough Float=0.25)
    Sprite Rotation Mode=::NewEnumerator1
    Sprite Size Mode=::NewEnumerator3
    Uniform Sprite Size=&MaskFloatBySpawnGroup(
     Masked Float=&RandomRangeFloat(Minimum=20, Maximum=25)
     Passthrough Float=&MaskFloatBySpawnGroup(
      Masked Float=&RandomRangeFloat(Minimum=5, Maximum=10)
      Passthrough Float=&RandomRangeFloat(Minimum=25, Maximum=30)
      Spawn Group Mask=1
     )
    )
    UsePositionOffset=false
    Write Material Random=true
   )
   &CameraOffset(Camera Offset Amount=&FN_FrontEndFloatModifier(Gameplay=0, Front End=0.25))
   &InitialMeshOrientation(Mesh Orientation Mode=::NewEnumerator2, Use Rotation Vector=false)
   &SkeletalMeshLocation(
    Bone / Socket Index=1
    Bone Sampling Mode=::NewEnumerator4
    Interpolate Mesh Sampling=false
    MaskSpawnGroup=false
    Orientation Sampling=::NewEnumerator1
    Sampled Position Offset=&SelectVectorFromArray(
     Direct Array Index=&SpawnGroup()
     Vector Selection Array=DataInterface() {propertyValues {FloatData [<0.35 3.5 2.25>, <1 0.8 -7.5>, <1.25 0.2 -0.1>]}}
    )
    Skeletal Mesh=$System.Skeletal Mesh
   ) {module=SkeletalMeshLocation001}
  ]
  particleUpdateScript [
   &ParticleState()
   &SkeletalMeshLocation(
    Bone / Socket Index=1
    Bone Sampling Mode=::NewEnumerator4
    MaskSpawnGroup=false
    Orientation Sampling=::NewEnumerator1
    Position Sampling=::NewEnumerator2
    Skeletal Mesh=$System.Skeletal Mesh
   )
   &SpriteFacingAndAlignment(
    Sprite Facing=&RotateVectorByQuaternion(Vector To Rotate=<1 0.21 0>, Quaternion=$Particles.SkeletalMeshLocation.SampledRotation)
    WriteAlignment=true
    WriteFacing=true
   )
   &DynamicMaterialParameters(
    Index 0 Param 1=&FloatFromCurve(
     FloatCurve=DataInterface() {propertyValues {Curve {keys [
         {arriveTangent=6.1038, interpMode=RCIM_Cubic, leaveTangent=6.1038, tangentMode=RCTM_User, time=0, value=0.5}
         {arriveTangent=0.038, interpMode=RCIM_Cubic, leaveTangent=0.038, tangentMode=RCTM_User, time=0.2, value=1}
         {arriveTangent=-0.0405, interpMode=RCIM_Cubic, leaveTangent=-0.0405, tangentMode=RCTM_User, time=1, value=0.5}
        ]}}}
     Scale Curve=&MaskFloatBySpawnGroup(
      Masked Float=0.4
      Passthrough Float=&MaskFloatBySpawnGroup(Masked Float=0.8, Passthrough Float=0.3, Spawn Group Mask=1)
     )
    )
    Index 0 Param 2=&FloatFromCurve(
     FloatCurve=DataInterface() {propertyValues {Curve {keys [
         {arriveTangent=2.0212, interpMode=RCIM_Cubic, leaveTangent=2.0212, tangentMode=RCTM_User, time=0, value=0}
         {interpMode=RCIM_Cubic, tangentMode=RCTM_User, time=1, value=1}
        ]}}}
     Scale Curve=&MaskFloatBySpawnGroup(Masked Float=0.25, Passthrough Float=0.5, Spawn Group Mask=2)
    )
    Index 0 Param 3=&MaskFloatBySpawnGroup(
     Masked Float=&FloatFromCurve(
      FloatCurve=DataInterface() {propertyValues {Curve {keys [{interpMode=RCIM_Cubic, tangentMode=RCTM_User, time=0, value=0.55}, {interpMode=RCIM_Cubic, tangentMode=RCTM_User, time=1, value=1}]}}}
      Scale Curve=0.3
     )
     Passthrough Float=&MaskFloatBySpawnGroup(
      Masked Float=&FloatFromCurve(
       FloatCurve=DataInterface() {propertyValues {Curve {keys [{interpMode=RCIM_Cubic, tangentMode=RCTM_User, time=0, value=0.3}, {interpMode=RCIM_Cubic, tangentMode=RCTM_User, time=1, value=1}]}}}
       Scale Curve=0.3
      )
      Passthrough Float=&FloatFromCurve(
       FloatCurve=DataInterface() {propertyValues {Curve {keys [
           {arriveTangent=-0.133, interpMode=RCIM_Cubic, leaveTangent=-0.133, tangentMode=RCTM_User, time=0, value=0.9}
           {arriveTangent=-0.0015, interpMode=RCIM_Cubic, leaveTangent=-0.0015, tangentMode=RCTM_User, time=1, value=0.785}
          ]}}}
       Scale Curve=0.3
      )
      Spawn Group Mask=1
     )
    )
    Index 0 Param 4=1
    Index 1 Param 1=&MaskFloatBySpawnGroup(Masked Float=0.45, Passthrough Float=0.5)
    Index 1 Param 2=&MaskFloatBySpawnGroup(Masked Float=0.56, Passthrough Float=-0.1)
    Index 1 Param 3=&MaskFloatBySpawnGroup(
     Masked Float=0.3
     Passthrough Float=&MaskFloatBySpawnGroup(Masked Float=0.7, Passthrough Float=2, Spawn Group Mask=1)
    )
    Module.Index0Param0Write=true
    Module.Index0Param1Write=true
    Module.Index0Param2Write=true
    Module.Index0Param3Write=true
    Module.Index1Param0Write=true
    Module.Index1Param1Write=true
    Module.Index1Param2Write=true
    Module.Index1Param3Write=true
    Write Parameter Index 1=true
   )
   &ScaleColor(
    Scale Alpha=&FloatFromCurve(FloatCurve=DataInterface() {propertyValues {Curve {keys [
         {interpMode=RCIM_Cubic, tangentMode=RCTM_User, time=0, value=0}
         {interpMode=RCIM_Cubic, tangentMode=RCTM_User, time=0.5, value=1}
         {tangentMode=RCTM_User, time=1, value=0}
        ]}}})
    Scale RGB=&MakeVectorFromLinearColorRGB(Color=&FNTimeOfDayColor(
      Night Color=<rgba 0.26 0.4943 1 1>
      Morning Color=<rgba 0.8 0.5397 0.516 1>
      Day Color=<rgba 1 1 1 1>
      Evening Color=<rgba 0.8021 0.5013 0.5013 1>
     ))
    ScaleA=false
    ScaleRGB=true
   )
   /SetVariables(Particles.Ripple Color:LinearColor=&MakeLinearColorFromVectorAndFloat(
     "Vector (RGB)"=&MakeVectorFromLinearColorRGB(Color=$Particles.Color)
     "Float (Alpha)"=&FloatFromCurve(
      FloatCurve=DataInterface() {propertyValues {Curve {keys [
          {arriveTangent=7.0005, interpMode=RCIM_Cubic, leaveTangent=7.0005, tangentMode=RCTM_User, time=0, value=0}
          {interpMode=RCIM_Cubic, tangentMode=RCTM_User, time=0.2, value=1}
          {arriveTangent=0.006, interpMode=RCIM_Cubic, leaveTangent=0.006, tangentMode=RCTM_User, time=0.7, value=1}
          {tangentMode=RCTM_User, time=1, value=0}
         ]}}}
      Scale Curve=&RandomRangeFloat(Minimum=0.2, Maximum=0.3)
     )
    )) {isSetParameters=true, module=SetVariables_#3}
  ]
  rendererClasses [&NiagaraSpriteRendererProperties]
  renderers [{rendererClass=&NiagaraSpriteRendererProperties}]
 }
}
summary {
 systemName=NS_Backpack_HayBrush_Idle
 emitters [
  {
   emitterName=Bubble_Ripple_Burst
   rendererClasses [&NiagaraMeshRendererProperties, &NiagaraSpriteRendererProperties]
  }
  {emitterName=Ripples_Frog_and_Spoon, rendererClasses [&NiagaraSpriteRendererProperties]}
 ]
 userVariables [
  User.bIsFrontEnd {type=NiagaraBool}
  User.bIsFrontEndPreview {type=NiagaraBool}
 ]
}
