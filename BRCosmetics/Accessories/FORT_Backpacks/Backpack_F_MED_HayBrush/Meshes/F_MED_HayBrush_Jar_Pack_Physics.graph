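# Physics asset for the F_MED_HayBrush jar backpack cosmetic, giving its pack, cauldron base and jar bones capsule bodies for physics.
# The bodies form a chain: dyn_jar is constrained to base_cauldron with tight swing and twist limits, allowing only slight wobble.
# The base_cauldron joint to pack leaves all angular axes free.
# The cauldron combines two capsules, one offset and angled, while the pack has just a tiny capsule.
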PhysicsAsset /BRCosmetics/Accessories/FORT_Backpacks/Backpack_F_MED_HayBrush/Meshes/F_MED_HayBrush_Jar_Pack_Physics
bodies {
 base_cauldron [
  {shapeType=Capsule, length=1, radius=1, shapeName=base_cauldron_capsule, rotation=<rot 0 0 -90>}
  {
   shapeType=Capsule
   length=5.5
   radius=3
   shapeName=base_cauldron_capsule
   center=<-2.3979 -10.589 5.8504>
   rotation=<rot 75.4676 -16.433 -10.4476>
  }
 ]
 dyn_jar [{shapeType=Capsule, length=1.0582, radius=5.5434, shapeName=dyn_jar_capsule, center=<3.6642 -0.0906 -0.086>, rotation=<rot -90 0 0>}]
 pack [{shapeType=Capsule, length=0.505, radius=0.505, shapeName=pack_capsule, rotation=<rot 0 -90 0>}]
}
constraints [
 {bone1Name=dyn_jar, bone2Name=base_cauldron, swing1LimitDegrees=10, swing2LimitDegrees=15, twistLimitDegrees=5}
 {
  bone1Name=base_cauldron
  bone2Name=pack
  swing1LimitDegrees=45
  swing1Motion=Free
  swing2LimitDegrees=45
  swing2Motion=Free
  twistLimitDegrees=45
  twistMotion=Free
 }
]
