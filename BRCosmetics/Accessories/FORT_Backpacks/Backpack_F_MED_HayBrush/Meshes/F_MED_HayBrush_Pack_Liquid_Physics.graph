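# Physics asset for the liquid mesh of the F_MED_HayBrush cosmetic backpack, likely driving potion motion in a cauldron.
# Small capsule bodies sit on the pack, base_cauldron and dyn_potion bones, linked in a chain by two constraints.
# The cauldron rotates freely relative to the pack, while the potion has a very narrow swing range and zero twist limit against the cauldron.
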PhysicsAsset /BRCosmetics/Accessories/FORT_Backpacks/Backpack_F_MED_HayBrush/Meshes/F_MED_HayBrush_Pack_Liquid_Physics
bodies {
 base_cauldron [{shapeType=Capsule, length=1, radius=1, shapeName=base_cauldron_capsule, rotation=<rot 0 0 -90>}]
 dyn_potion [{shapeType=Capsule, length=0.5, radius=0.5, shapeName=dyn_potion_capsule, center=<0.1 0 0>, rotation=<rot 0 -90 0>}]
 pack [{shapeType=Capsule, length=0.505, radius=0.505, shapeName=pack_capsule, rotation=<rot 0 -90 0>}]
}
constraints [
 {bone1Name=dyn_potion, bone2Name=base_cauldron, swing1LimitDegrees=2.1, swing2LimitDegrees=2.1, twistLimitDegrees=0}
 {
  bone1Name=base_cauldron
  bone2Name=pack
  swing1LimitDegrees=45
  swing1Motion=Free
  swing2LimitDegrees=45
  swing2Motion=Free
  twistLimitDegrees=45
  twistMotion=Free
 }
]
