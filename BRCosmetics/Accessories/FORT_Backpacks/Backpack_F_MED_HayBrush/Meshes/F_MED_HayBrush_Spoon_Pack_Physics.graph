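# Physics asset for the HayBrush spoon backpack cosmetic, defining collision bodies and joints for its cauldron and spoon.
# Four capsule bodies form a chain from pack through base_cauldron and dyn_cauldron to dyn_spoon, with dyn_cauldron the widest.
# The cauldron joint allows only very slight swing and twist, while the spoon joint allows a somewhat wider, asymmetric wobble.
# The base_cauldron-to-pack joint leaves swing and twist free.
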PhysicsAsset /BRCosmetics/Accessories/FORT_Backpacks/Backpack_F_MED_HayBrush/Meshes/F_MED_HayBrush_Spoon_Pack_Physics
bodies {
 base_cauldron [{shapeType=Capsule, length=1, radius=1, shapeName=base_cauldron_capsule, rotation=<rot 0 0 -90>}]
 dyn_cauldron [{shapeType=Capsule, length=0, radius=11.5, shapeName=dyn_cauldron_capsule, center=<-0.6133 0 -0.5503>, rotation=<rot -90 180 180>}]
 dyn_spoon [{
   shapeType=Capsule
   length=8.4664
   radius=1.0275
   shapeName=dyn_spoon_capsule
   center=<7.2504 -0.197 -0.5359>
   rotation=<rot 0 -91.5114 -77.7898>
  }]
 pack [{shapeType=Capsule, length=0.505, radius=0.505, shapeName=pack_capsule, rotation=<rot 0 -90 0>}]
}
constraints [
 {bone1Name=dyn_cauldron, bone2Name=base_cauldron, swing1LimitDegrees=2.1, swing2LimitDegrees=2.1, twistLimitDegrees=2}
 {bone1Name=dyn_spoon, bone2Name=dyn_cauldron, swing1LimitDegrees=15, swing2LimitDegrees=8.5, twistLimitDegrees=5}
 {
  bone1Name=base_cauldron
  bone2Name=pack
  swing1LimitDegrees=45
  swing1Motion=Free
  swing2LimitDegrees=45
  swing2Motion=Free
  twistLimitDegrees=45
  twistMotion=Free
 }
]
