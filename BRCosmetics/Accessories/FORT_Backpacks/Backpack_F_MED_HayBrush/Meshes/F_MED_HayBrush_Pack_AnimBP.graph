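# Animation Blueprint for the HayBrush cauldron backpack cosmetic, playing an idle and a one-shot reaction to shield gains or health increases.
# The state machine's pose gets a cauldron tilt when downed, then passes through chained spoon, liquid and jar RigidBody simulations, plus a jar offset while jumping.
# Animation curves weight liquid and jar physics; parachuting softens spoon physics.
# The event graph also sets gravity, frontend/in-game sim settings and gimbal rotations.
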AnimBlueprint /BRCosmetics/Accessories/FORT_Backpacks/Backpack_F_MED_HayBrush/Meshes/F_MED_HayBrush_Pack_AnimBP
target_skeleton=&/F_MED_HayBrush_Pack_Skeleton
graphs [
 AnimGraph {nodes [
   Input Pose()
   n1 = "Transform (Modify) Bone - Bone: base_cauldron"(ComponentPose=Local To Component(LocalPose=Use cached pose 'StartCache'()), Alpha=1, Translation="0,0,0", Rotation="-90,0,0", Scale="1,1,1") {
    rotationMode=BMM_Additive
    rotationSpace=BCS_BoneSpace
    alphaBoolBlend {blendOption=Linear}
    boneToModify {boneName=base_cauldron}
    rotation {pitch=-90}
   }
   n2 = Component To Local(ComponentPose=@n1)
   n3 = Get bIsDBNO()
   n4 = "Transform (Modify) Bone - Bone: base_cauldron"(ComponentPose=Local To Component(LocalPose=Use cached pose 'StartCache'()), Alpha=1, Translation="0,0,0", Rotation="0,0,0", Scale="1,1,1") {
    rotationMode=BMM_Additive
    rotationSpace=BCS_BoneSpace
    alphaBoolBlend {blendOption=Linear}
    alphaScaleBiasClamp {bInterpResult=true}
    boneToModify {boneName=base_cauldron}
   }
   n5 = "Transform (Modify) Bone - Bone: base_cauldron"!(ComponentPose=@n4, bAlphaBoolEnabled:bool=True, Translation="0,0,0", Rotation="0,0,0", Scale="1,1,1") -> (Pose) {
    alphaInputType=Bool
    rotationMode=BMM_Additive
    rotationSpace=BCS_BoneSpace
    alphaBoolBlend {blendOption=QuarticInOut, blendOutTime=10}
    alphaScaleBiasClamp {bInterpResult=true}
    boneToModify {boneName=base_cauldron}
   }
   n6 = Component To Local(ComponentPose=@n5)
   n7 = RigidBody!(
    ComponentPose=Local To Component(LocalPose=Blend Poses by bool(BlendPose_0=@n2, BlendPose_1=@n6, bActiveValue=@n3) {blendType=Linear})
    Alpha=Get RigidBody Alpha() -> (RigidBody Alpha:double)
    OverrideWorldGravity:Vector=Get Gravity Override()
    ExternalForce
    ComponentLinearAccScale:Vector="0.6,0.6,0.8"
    ComponentLinearVelScale:Vector="0.6,0.6,0.8"
    SimSpaceSettings:Sim Space Settings Structure=Get Sim Space Settings() -> (Sim Space Settings:Sim Space Settings Structure)
   ) -> (Pose) {
    bClampLinearTranslationLimitToRefPose=true
    bOverrideWorldGravity=true
    overridePhysicsAsset=&/F_MED_HayBrush_Spoon_Pack_Physics
    simulationSpace=BaseBoneSpace
    alphaBoolBlend {blendOption=Linear}
    alphaScaleBiasClamp {bInterpResult=true}
    baseBoneRef {boneName=root}
    componentAppliedLinearAccClamp=<3000 3000 3000>
    componentLinearAccScale=<0.3 0.2 0.2>
    componentLinearVelScale=<0.3 0.2 0.2>
    simSpaceSettings {
     maxAngularAcceleration=1000
     maxAngularVelocity=3
     maxLinearAcceleration=700
     maxLinearVelocity=500
     velocityScaleZ=0
     worldAlpha=0.5
     externalLinearDragV=<0.2 0.2>
    }
   }
   n8 = RigidBody(
    ComponentPose=@n7
    Alpha=Get Curve_Liquid_Alpha() -> (Curve_Liquid_Alpha:double)
    OverrideWorldGravity=Get Gravity Override()
    ComponentLinearAccScale="0.6,0.6,0.8"
    ComponentLinearVelScale="0.6,0.6,0.8"
    SimSpaceSettings=Get Sim Space Settings()
   ) {
    bClampLinearTranslationLimitToRefPose=true
    bOverrideWorldGravity=true
    overridePhysicsAsset=&/F_MED_HayBrush_Pack_Liquid_Physics
    simulationSpace=BaseBoneSpace
    alphaBoolBlend {blendOption=Linear}
    alphaScaleBiasClamp {bInterpResult=true}
    baseBoneRef {boneName=root}
    componentAppliedLinearAccClamp=<3000 3000 3000>
    componentLinearAccScale=<0.3 0.2 0.2>
    componentLinearVelScale=<0.3 0.2 0.2>
    simSpaceSettings {
     maxAngularAcceleration=1000
     maxAngularVelocity=3
     maxLinearAcceleration=700
     maxLinearVelocity=500
     velocityScaleZ=0
     worldAlpha=0.5
     externalLinearDragV=<0.2 0.2>
    }
   }
   n9 = RigidBody(
    ComponentPose=@n8
    Alpha=Get Curve_Jar_Alpha() -> (Curve_Jar_Alpha:double)
    OverrideWorldGravity=Get Gravity Override()
    ComponentLinearAccScale="0.6,0.6,0.8"
    ComponentLinearVelScale="0.6,0.6,0.8"
    SimSpaceSettings=Get Sim Space Settings()
   ) {
    bClampLinearTranslationLimitToRefPose=true
    bOverrideWorldGravity=true
    overridePhysicsAsset=&/F_MED_HayBrush_Jar_Pack_Physics
    simulationSpace=BaseBoneSpace
    alphaBoolBlend {blendOption=Linear}
    alphaScaleBiasClamp {bInterpResult=true}
    baseBoneRef {boneName=root}
    componentAppliedLinearAccClamp=<3000 3000 3000>
    componentLinearAccScale=<0.3 0.2 0.2>
    componentLinearVelScale=<0.3 0.2 0.2>
    simSpaceSettings {
     maxAngularAcceleration=1000
     maxAngularVelocity=3
     maxLinearAcceleration=700
     maxLinearVelocity=500
     velocityScaleZ=0
     worldAlpha=0.5
     externalLinearDragV=<0.2 0.2>
    }
   }
   n10 = "Transform (Modify) Bone - Bone: dyn_jar"(ComponentPose=@n9, Alpha=Get Is Jumping Float() -> (Is Jumping Float:double), Translation="2,0,1.5", Rotation="-15,0,0", Scale="1,1,1") {
    rotationMode=BMM_Additive
    rotationSpace=BCS_BoneSpace
    translationMode=BMM_Additive
    translationSpace=BCS_BoneSpace
    alphaBoolBlend {blendOption=Linear}
    alphaScaleBiasClamp {bInterpResult=true, interpSpeedDecreasing=2.5, interpSpeedIncreasing=2.5}
    boneToModify {boneName=dyn_jar}
    rotation {pitch=-15}
    translation {x=2, z=1.5}
   }
   n11 = "Transform (Modify) Bone - Bone: dyn_jar"(ComponentPose=@n10, Alpha=1, Translation="0,0,0", Rotation="0,0,0", Scale="1,1,1") {
    rotationMode=BMM_Additive
    rotationSpace=BCS_BoneSpace
    translationMode=BMM_Additive
    translationSpace=BCS_BoneSpace
    alphaBoolBlend {blendOption=QuinticInOut, blendOutTime=10}
    alphaScaleBiasClamp {bInterpResult=true, interpSpeedDecreasing=2.5, interpSpeedIncreasing=2.5}
    boneToModify {boneName=dyn_jar}
   }
   Output Pose(Result=Component To Local(ComponentPose=@n11)) {layerGroup=None}
   AnimGraphNode_StateMachine_0 = Backpack States() {is=title AnimGraphNode_StateMachine}
   Save cached pose 'StartCache'(Pose=@AnimGraphNode_StateMachine_0) {cacheName=StartCache}
  ]}
 Backpack States {
  entry=Idle_01
  path=AnimGraph.AnimGraphNode_StateMachine_0.Backpack States
  ids=id
  nodes [
   AnimStateNode_0 = Idle_Random() {
    is=title AnimStateNode
    stateEntered {guid=#1, notifyName=Entered_RandomLocoBreak}
    stateFullyBlended {guid=#2}
    stateLeft {guid=#3}
   }
   AnimStateNode_1 = Idle_01() {is=title AnimStateNode}
   AnimStateNode_2 = Reaction() {is=title AnimStateNode}
   AnimStateTransitionNode_2 = AnimStateTransitionNode()
   AnimStateTransitionNode_3 = AnimStateTransitionNode() {bAutomaticRuleBasedOnSequencePlayerInState=true}
  ]
  links [
   {
    dst=Reaction
    rule=AnimGraph.AnimGraphNode_StateMachine_0.Backpack States.AnimStateTransitionNode_2.Transition
    src=Idle_01
    via=AnimStateTransitionNode_2
   }
   {
    dst=Idle_01
    rule=AnimGraph.AnimGraphNode_StateMachine_0.Backpack States.AnimStateTransitionNode_3.Transition
    src=Reaction
    via=AnimStateTransitionNode_3
   }
  ]
 }
 Idle_Random {
  path=AnimGraph.AnimGraphNode_StateMachine_0.Backpack States.AnimStateNode_0.Idle_Random
  nodes [
   n1 = Random Sequence Player!() -> (Pose:Pose Link Structure) {
    is=title AnimGraphNode_RandomPlayer
    entries [
     {
      chanceToPlay=0.7
      maxLoopCount=0
      maxPlayRate=1
      minLoopCount=0
      minPlayRate=1
      sequence=/BRCosmetics/Accessories/FORT_Backpacks/Backpack_F_MED_HayBrush/Anim/Backpack_HayBrush_IDLE
      blendIn {blendOption=Linear, blendTime=0.2, customCurve=null}
     }
     {
      chanceToPlay=0.5
      maxLoopCount=0
      maxPlayRate=1
      minLoopCount=0
      minPlayRate=1
      sequence=null
      blendIn {blendOption=Linear, blendTime=0.2, customCurve=null}
     }
     {
      chanceToPlay=0.4
      maxLoopCount=0
      maxPlayRate=1
      minLoopCount=0
      minPlayRate=1
      sequence=&Backpack_HayBrush_IDLE
      blendIn {blendOption=Linear, blendTime=0.2, customCurve=null}
     }
    ]
   }
   Output Animation Pose(Result=@n1)
  ]
 }
 Idle_01 {
  path=AnimGraph.AnimGraphNode_StateMachine_0.Backpack States.AnimStateNode_1.Idle_01
  nodes [Output Animation Pose(Result=Sequence Player 'Backpack_HayBrush_IDLE'() {sequence=&Backpack_HayBrush_IDLE})]
 }
 Reaction {
  path=AnimGraph.AnimGraphNode_StateMachine_0.Backpack States.AnimStateNode_2.Reaction
  nodes [Output Animation Pose(Result=Sequence Player 'Backpack_HayBrush_REACTION'() {bLoopAnimation=false, sequence=/BRCosmetics/Accessories/FORT_Backpacks/Backpack_F_MED_HayBrush/Anim/Backpack_HayBrush_REACTION})]
 }
 Transition {
  path=AnimGraph.AnimGraphNode_StateMachine_0.Backpack States.AnimStateTransitionNode_2.Transition
  nodes [Result(bCanEnterTransition=Get GainShield() -> (GainShield:bool))]
 }
 Transition {path=AnimGraph.AnimGraphNode_StateMachine_0.Backpack States.AnimStateTransitionNode_3.Transition, nodes [Result()]}
 EventGraph {nodes [
   Event BlueprintPostEvaluateAnimation() {is=title K2Node_Event}
   > Set GainShield(GainShield:bool=false) -> (Output_Get:bool)
   Event BlueprintInitializeAnimation()
   > n1 = Cast To FortPlayerPawn(Object=TryGetFortPawn())
   > "Set Skel mesh "("Skel mesh ":Skeletal Mesh Component Object Reference=GetSkeletalMeshForPartType(self=@n1)) -> (Output_Get:Skeletal Mesh Component Object Reference)
   > n2 = WatchCosmeticStat!(
    execute:Exec
    Pawn:Fort Player Pawn Object Reference=@n1
    StatTag:Fort Stat Manager Tag Structure=(TagName="StatManager.GainedShieldTimes")
    DefaultValue:int=0
    bBroadcastInitialValue:bool=false
    InitialValueBroadcastDelay:float=-1
    ChangeValueBroadcastDelay:float=-1
    bConsumeWaitingNotifications:bool=true
   ) -> (then:Exec, OnStatChanged:Exec, StatName:Name, StatValue:int) {is=title K2Node_AsyncAction}
   Branch(execute=[@n2.then, @n18.then], Condition=Get bShouldDisplayFrontendReactivity() -> (bShouldDisplayFrontendReactivity:bool))
   > Delay(Duration=10)
   > Set GainShield(GainShield=true)
   Set GainShield(execute=[@n2.OnStatChanged, @n20.then], GainShield=true)
   > RetriggerableDelay(Duration=0.2)
   > Set GainShield(GainShield=false)
   Event BlueprintUpdateAnimation()
   > n3 = Sequence!(execute) -> (then_0, then_1, then_2:Exec)
   Set Curve_Jar_Alpha(execute=@n3.then_2, Curve_Jar_Alpha:double=GetCurveValue(CurveName=Jar Alpha)) -> (Output_Get:double)
   > Set Curve_Liquid_Alpha(Curve_Liquid_Alpha:double=GetCurveValue(CurveName=Liquid Alpha)) -> (Output_Get:double)
   Branch(execute=@n3.then_1, Condition=NOT Boolean(A=Get bIsSkyDiving()))
   > Set Rot Down(Rot Down:float=GetAlphaFromReferenceTransformDelta(
     SkeletalMeshComponent="Get Skel mesh "() -> ("Skel mesh ":Skeletal Mesh Component Object Reference)
     BoneName=spine_03
     BaseName="head "
     InputRange=(X=0,Y=-120)
     OutputRange=(X=0,Y=-1)
     TwistAxis="0,0,1"
    )) -> (Output_Get:float)
   > Set Rot Up(Rot Up:float=GetAlphaFromReferenceTransformDelta(SkeletalMeshComponent="Get Skel mesh "(), BoneName=spine_04, BaseName=head, InputRange=(X=15,Y=60), TwistAxis="0,0,1")) -> (Output_Get:float)
   n4 = GetRealTimeSeconds()
   n5 = "Sin (Radians)"(A=float * float(A=float * float(A=1, B=1.4) {is=title K2Node_PromotableOperator}, B=@n4))
   n6 = "Sin (Radians)"(A=float * float(A=float * float(A=1, B=0.8) {is=title K2Node_PromotableOperator}, B=@n4) {is=title K2Node_CommutativeAssociativeBinaryOperator})
   n7 = "Sin (Radians)"(A=float * float(A=float * float(A=1, B=2) {is=title K2Node_PromotableOperator}, B=@n4) {is=title K2Node_CommutativeAssociativeBinaryOperator})
   > Set Jar Rot(Jar Rot:Rotator=MakeRotator(Roll=float / float(A=@n5, B=-0.25), Pitch=float / float(A=@n7, B=-0.5), Yaw=float / float(A=@n6, B=-0.75))) -> (Output_Get:Rotator)
   n8 = Gravity Override State Function DEPRECATED(
    execute=@n3.then_0
    CharacterPartAnimInstance=Self-Reference()
    GameTimeInSeconds=GetGameTimeInSeconds()
    GravityOverrideParams=Get Gravity Override Params()
   )
   > Set Gravity Override(Gravity Override=@n8)
   n9 = GetOwnerSkeletalMeshComponent()
   n10 = BreakRotator(InRot=BreakTransform(InTransform=GetDeltaTransformFromRefPose(self=@n9, BoneName=spine_05, BaseName=root)).Rotation)
   n11 = float * float(A=@n10.Yaw, B=Get Rot Down() -> (Rot Down:float))
   > Set Gimbal Rot(Gimbal Rot:Rotator=MakeRotator(Roll=float * float(A=@n11, B=0.5), Pitch=float * float(A=@n10.Yaw, B=-0.8))) -> (Output_Get:Rotator)
   > Branch(Condition=NOT Boolean(A=Get bIsCrouching()))
   > Set ADSGimbal(ADSGimbal:Rotator=MakeRotator(Roll=float * float(A=@n10.Yaw, B=-1), Pitch=float * float(A=@n10.Yaw, B=0.3))) -> (Output_Get:Rotator)
   n12 = NOT Boolean(A=Get bIsSkyDiving())
   n13 = Get bIsJumping()
   > Set Is Jumping Float(Is Jumping Float:double="To Float (Boolean)"(InBool=AND Boolean(A=@n12, B=@n13))) -> (Output_Get:double)
   > n14 = Branch(Condition=Get bIsParachuteOpened())
   > n15 = Set RigidBody Alpha(RigidBody Alpha:double=0.25) -> (Output_Get:double)
   n16 = Branch(execute=[@n17.then, @n15.then], Condition=Get bShouldDisplayFrontendReactivity())
   > Set Sim Space Settings(Sim Space Settings:Sim Space Settings Structure=Get FrontEnd Sim Space Settings() -> (FrontEnd Sim Space Settings:Sim Space Settings Structure)) -> (Output_Get:Sim Space Settings Structure)
   Set Sim Space Settings(execute=@n16.else, Sim Space Settings=Get InZone Sim Space Settings() -> (InZone Sim Space Settings:Sim Space Settings Structure))
   n17 = Set RigidBody Alpha(execute=@n14.else, RigidBody Alpha=1)
   AnimNotify_ReactiveReset() {is=title K2Node_Event}
   > n18 = Set GainShield(GainShield=false)
   n19 = Event BlueprintProcessAnimInputEvent()
   > n20 = Branch(Condition=IsNamed(self=@n19.AnimInputEvent, InEventName="aie.PlayerVitality.HealthIncreased\t"))
  ]}
]
variables [
 ADSGimbal {type=Rotator}
 Curve_Jar_Alpha {type=number}
 Curve_Liquid_Alpha {type=number}
 FrontEnd Sim Space Settings {
  type=SimSpaceSettings
  default {
   dampingAlpha=1
   maxAngularAcceleration=5000
   maxAngularVelocity=5000
   maxLinearAcceleration=5000
   maxLinearVelocity=5000
   velocityScaleZ=0.25
   worldAlpha=0.35
   externalLinearDragV=<0 0 0>
  }
 }
 GainShield {type=boolean}
 Gimbal Rot {type=Rotator}
 Gravity Override {type=Vector}
 Gravity Override Params {type=GravityOverrideParamsStruct}
 InZone Sim Space Settings {
  type=SimSpaceSettings
  default {
   dampingAlpha=1
   maxAngularAcceleration=10000
   maxAngularVelocity=10000
   maxLinearAcceleration=10000
   maxLinearVelocity=10000
   velocityScaleZ=1
   worldAlpha=0
   externalLinearDragV=<0 0 0>
  }
 }
 Is Jumping Float {type=number}
 Jar Rot {type=Rotator}
 RigidBody Alpha {type=number}
 Rot Down {type=number}
 Rot Up {type=number}
 Sim Space Settings {
  type=SimSpaceSettings
  default {
   dampingAlpha=1
   maxAngularAcceleration=10000
   maxAngularVelocity=10000
   maxLinearAcceleration=10000
   maxLinearVelocity=10000
   velocityScaleZ=1
   worldAlpha=0
   externalLinearDragV=<0 0 0>
  }
 }
 "Skel mesh " {type=SkeletalMeshComponent}
]
